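# Material function for the ShinyStar pickaxe that samples a beam mask texture in radial coordinates to produce animated, rotating beams, output as RGB.
# UVs blend toward a bump-offset parallax version by a mask, add a distortion input, then convert to radial space around an offset center.
# RotateSpeed scrolls the first radial coordinate, while the Time input shifts the second.
# Sampling uses explicit derivative inputs, previewed from unwarped UVs, instead of derivatives of the radial UVs.
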MaterialFunction /BRCosmetics/Weapons/FORT_Melee/Pickaxe_ShinyStar/Materials/MF_ShinyStar_Beam
graph {nodes [
  n1 = Constant() {r=0.1}
  n2 = Frac(None=Multiply(A=Time(), B=FunctionInput(Preview=@n1) {inputName=RotateSpeed, inputType=FunctionInput_Scalar}))
  n3 = TextureCoordinate()
  n4 = FunctionInput(Preview=Constant() {r=2}) {inputName=BumpOffset, inputType=FunctionInput_Scalar}
  n5 = Constant() {r=1}
  n6 = Add(
   A=LinearInterpolate(A=@n3, B=BumpOffset(Coordinate=@n3, Height=@n4), Alpha=FunctionInput(Preview=@n5) {inputName=BumpOffsetMask, inputType=FunctionInput_Scalar})
   B=FunctionInput() {inputName=UV_Distortion, inputType=FunctionInput_Vector2}
  )
  n7 = FunctionInput(Preview=Constant2Vector() {g=0.5, r=0.5}) {inputName=UV_Offset, inputType=FunctionInput_Vector2, sortPriority=1}
  n8 = MaterialFunctionCall(Float2=MaterialFunctionCall(Vector or UVs=Subtract(A=@n6, B=@n7)) {materialFunction=&VectorToRadialValue}.Radial Coordinates) {materialFunction=&BreakOutFloat2Components}
  n9 = FunctionInput() {inputName=Time, inputType=FunctionInput_Scalar, sortPriority=2}
  n10 = TextureCoordinate()
  n11 = TextureSample!(
   UVs=MaterialFunctionCall(X=Add(A=@n2, B=@n8.R), Y=Add(A=@n8.G, B=@n9)) {materialFunction=&MakeFloat2}.Result
   Tex=FunctionInput(Preview=TextureObject() {samplerType=SAMPLERTYPE_Masks, texture=/BRCosmetics/Weapons/FORT_Melee/Pickaxe_ShinyStar/Textures/T_Backpack_ShinyStar_Beams}) {inputName=Tex, inputType=FunctionInput_Texture2D, sortPriority=3}
   "DDX(UVs)"=FunctionInput(Preview=DDX(Value=@n10)) {inputName=DDX, inputType=FunctionInput_Scalar, sortPriority=4}
   "DDY(UVs)"=FunctionInput(Preview=DDY(Value=@n10)) {inputName=DDY, inputType=FunctionInput_Scalar, sortPriority=5}
   Apply View MipBias
  ) {
   desc=Radial Tex
   mipValueMode=TMVM_Derivative
   samplerSource=SSM_Wrap_WorldGroupSettings
   samplerType=SAMPLERTYPE_Masks
   texture=&T_Backpack_ShinyStar_Beams
  }
  FunctionOutput(None=@n11.RGB)
 ]}
