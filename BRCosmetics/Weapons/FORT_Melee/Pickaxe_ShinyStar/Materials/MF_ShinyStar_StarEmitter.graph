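# Material function for the ShinyStar pickaxe that emits one looping, spinning star, output as a green-channel mask.
# A wrapped panner moves the star along Direction, with Speed scaled by Reactivity and a time Offset input to shift its phase.
# The texture's green channel, rotated at RotationSpeed and clipped by a moving sphere mask, is thresholded as a distance field.
# The threshold blends between the PanningStar_DF min and max parameters through rerouted values.
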MaterialFunction /BRCosmetics/Weapons/FORT_Melee/Pickaxe_ShinyStar/Materials/MF_ShinyStar_StarEmitter
graph {nodes [
  n1 = Constant()
  n2 = ScalarParameter() {defaultValue=0.8, group=ShinyStar, parameterName=PanningStar_DF_Min_02}
  n3 = ScalarParameter() {defaultValue=0.45, group=ShinyStar, parameterName=PanningStar_DF_Min_01}
  n4 = NamedRerouteUsage()
  n5 = TextureCoordinate()
  n6 = Constant() {r=0.5}
  n7 = ComponentMask(None=VectorParameter() {group=ShinyStar, parameterName=ShinyStar_StarOrigin_UVOffset, defaultValue=<rgba -0.1 0.1 0 1>}.RGB) {g=true, r=true}
  n8 = Normalize(VectorInput=FunctionInput(Preview=Constant3Vector() {constant=<rgba 0.1 -0.1 0 1>}) {inputName=Direction, inputType=FunctionInput_Vector2, sortPriority=1})
  n9 = Multiply(
   A=FunctionInput(Preview=Constant() {r=0.2}) {inputName=Speed, inputType=FunctionInput_Scalar, sortPriority=2}
   B=NamedRerouteDeclaration(None=FunctionInput() {inputName=Reactivity, inputType=FunctionInput_Scalar}) {name=_reactivity, nodeColor=<rgba 1 0 0.8646 1>}
  )
  n10 = Frac(None=Panner(Coordinate=Subtract(A=@n6, B=@n7), Time=NamedRerouteUsage(), Speed=Multiply(A=@n8, B=@n9) {constB=0.5}))
  n11 = Subtract(A=Multiply(A=@n10) {constB=2})
  n12 = MaterialFunctionCall(Multiplier=FunctionInput(Preview=Constant() {r=1}) {inputName=RotationSpeed, inputType=FunctionInput_Scalar}) {materialFunction=&Pi}
  n13 = TextureSample(
   UVs=Rotator(
    Coordinate=Add(A=@n5, B=@n11)
    Time=MaterialFunctionCall(Speed=@n12.Result) {materialFunction=&TimeWithSpeedVariable}.Time
   ) {speed=1}
   Tex=FunctionInput(Preview=TextureObject() {samplerType=SAMPLERTYPE_Masks, texture=/BRCosmetics/Weapons/FORT_Melee/Pickaxe_ShinyStar/Textures/T_Backpack_ShinyStar_TA}) {inputName=Tex, inputType=FunctionInput_Texture2D, sortPriority=4}
  ) {
   samplerSource=SSM_Wrap_WorldGroupSettings
   samplerType=SAMPLERTYPE_Masks
   texture=&T_Backpack_ShinyStar_TA
  }
  n14 = Subtract(A=TextureCoordinate()) {constB=0.5}
  n15 = Subtract(A=Multiply(A=@n10) {constB=-2}) {constB=-1}
  n16 = Step(
   Y=LinearInterpolate(
    A=ScalarParameter() {defaultValue=0.8, group=ShinyStar, parameterName=PanningStar_DF_Max}
    B=LinearInterpolate(A=@n2, B=@n3, Alpha=@n4) {constA=0.45, constB=0.3}
    Alpha=NamedRerouteUsage()
   ) {constA=0.8, constB=0.45}
   X=Multiply(A=@n13.G, B=SphereMask(A=@n14, B=@n15) {attenuationRadius=0.12})
  ) {constX=0.5, constY=0.5}
  FunctionOutput(None=MaterialFunctionCall(X=@n1, Y=SmoothStep(Value=@n16) {constMax=0.8, constMin=0.5}, Z=@n1) {materialFunction=&MakeFloat3}.Result)
  n17 = Time()
  n18 = FunctionInput(Preview=Constant() {r=0.2}) {inputName=Offset, inputType=FunctionInput_Scalar, sortPriority=3}
  NamedRerouteDeclaration(None=Frac(None=Add(A=@n17, B=@n18))) {name=_time, nodeColor=<rgba 1 0.6797 0 1>}
 ]}
