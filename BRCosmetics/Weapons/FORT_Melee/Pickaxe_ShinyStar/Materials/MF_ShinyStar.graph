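# Material function for the Shiny Star pickaxe that adds an animated star-and-beam effect to incoming base color, emissive, roughness and metallic attributes.
# An FX mask separates the animated area and hue-shiftable outer ring.
# Three rotating beam layers use bump offset, Perlin distortion and explicit UV derivatives over a radial gradient.
# Five star-emitter calls add panning stars, with reactivity controls, center-star glow and optional time-of-day emissive boost.
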MaterialFunction /BRCosmetics/Weapons/FORT_Melee/Pickaxe_ShinyStar/Materials/MF_ShinyStar
graph {nodes [
  Constant()
  n1 = GetMaterialAttributes(None=FunctionInput() {inputType=FunctionInput_MaterialAttributes}) {attributeGetTypes [
    #1
    #2
    #3
    #4
   ]}
  n2 = SetMaterialAttributes!(
   MaterialAttributes=@n1.MaterialAttributes
   BaseColor=NamedRerouteUsage()
   EmissiveColor=NamedRerouteUsage()
   Roughness=NamedRerouteUsage()
   Metallic=NamedRerouteUsage()
  ) {attributeSetTypes [
    #1
    #2
    #3
    #4
   ]}
  FunctionOutput(None=@n2)
  n3 = SphereMask(A=TextureCoordinate(), B=NamedRerouteUsage()) {attenuationRadius=0.5, hardnessPercent=0}
  Multiply(A=@n3) {constB=1.1}
  NamedRerouteDeclaration(None=@n1.BaseColor) {name=BaseColor_IN, nodeColor=<rgba 0.3125 0 1 1>}
  NamedRerouteDeclaration(None=@n1.EmissiveColor) {name=Emissive_IN, nodeColor=<rgba 1 0.4453 0 1>}
  NamedRerouteDeclaration(None=@n1.Metallic) {name=Metallic_IN, nodeColor=<rgba 0 1 0.8125 1>}
  n4 = NamedRerouteUsage()
  n5 = Add(A=@n4, B=ScalarParameter() {defaultValue=0.3, parameterName=ShinyStarMetallicOffset})
  NamedRerouteDeclaration(None=LinearInterpolate(A=@n4, B=Saturate(None=@n5), Alpha=NamedRerouteUsage())) {name=Metallic_OUT, nodeColor=<rgba 1 0.0469 0 1>}
  n6 = TextureObjectParameter() {
   group=Skin FX
   parameterName=SkinFX_Mask
   samplerType=SAMPLERTYPE_Masks
   texture=/BRCosmetics/Accessories/FORT_Backpacks/Backpack_F_MED_ShinyStar/Textures/T_F_MED_ShinyStar_Backpack_FX
  }
  n7 = MaterialFunctionCall(Float4=FunctionInput(Preview=TextureSample(Tex=@n6).RGBA) {inputName=SkinFX, inputType=FunctionInput_Vector4}) {materialFunction=&BreakOutFloat4Components}
  n8 = NamedRerouteDeclaration(None=@n7.G) {name=AnimatedArea, nodeColor=<rgba 0.8047 0 1 1>}
  NamedRerouteDeclaration(None=Add(A=Add(A=@n7.R, B=@n8), B=NamedRerouteDeclaration(None=@n7.B) {name=OutterRing, nodeColor=<rgba 0.3828 0 1 1>})) {name=MetallicMask, nodeColor=<rgba 1 0 0.2109 1>}
  NamedRerouteDeclaration(None=ComponentMask(None=VectorParameter() {group=ShinyStar, parameterName=ShinyStar_UV_Offset, defaultValue=<rgba 0.305 0.697 0 1>}.RGB) {g=true, r=true}) {name=UV_Offset, nodeColor=<rgba 0.2422 1 0 1>}
  n9 = TextureCoordinate()
  NamedRerouteDeclaration(None=DDX(Value=@n9)) {name=_ddx, nodeColor=<rgba 0.5 1 0 1>}
  n10 = TextureObjectParameter() {
   group=Skin FX
   parameterName=ShinyStar_TA
   samplerType=SAMPLERTYPE_Masks
   texture=/BRCosmetics/Weapons/FORT_Melee/Pickaxe_ShinyStar/Textures/T_Backpack_ShinyStar_TA
  }
  NamedRerouteDeclaration(None=@n10) {name=TA_Mask, nodeColor=<rgba 0.4766 1 0 1>}
  n11 = Power(Base=Distance(A=TextureCoordinate(), B=NamedRerouteUsage()), Exp=Constant() {r=0.7}) {constExponent=0.5}
  n12 = ScalarParameter() {defaultValue=-5, group=ShinyStar, parameterName=BumpOffset_01}
  n13 = ScalarParameter() {defaultValue=0.08, group=ShinyStar, parameterName=ShinyStar_Beam_RotateSpeed_01}
  n14 = TextureSample(UVs=Rotator(), Tex=NamedRerouteUsage()) {
   samplerSource=SSM_Wrap_WorldGroupSettings
   samplerType=SAMPLERTYPE_LinearGrayscale
   texture=&T_SimplePerlin_Greyscale
  }
  n15 = Saturate(None=SmoothStep(Value=Multiply(A=@n11) {constB=5}))
  n16 = MaterialFunctionCall(
   X=Multiply(A=Multiply(A=@n14.R, B=@n15), B=ScalarParameter() {defaultValue=0.085, group=ShinyStar, parameterName=DistortNoiseIntensity})
   Y=Constant()
  ) {desc=Distortion, materialFunction=&MakeFloat2}
  n17 = NamedRerouteUsage()
  n18 = NamedRerouteUsage()
  n19 = TextureObjectParameter() {
   group=ShinyStar
   parameterName=ShinyStar_Beams_Tex
   samplerType=SAMPLERTYPE_Masks
   texture=/BRCosmetics/Weapons/FORT_Melee/Pickaxe_ShinyStar/Textures/T_Backpack_ShinyStar_Beams
  }
  n20 = NamedRerouteUsage()
  n21 = NamedRerouteUsage()
  NamedRerouteDeclaration(None=ComponentMask(None=MaterialFunctionCall!(BumpOffsetMask=@n11, BumpOffset=@n12, RotateSpeed=@n13, UV_Distortion=@n16.Result, UV_Offset=@n17, Time=@n18, Tex=@n19, DDX=@n20, DDY=@n21) {materialFunction=&/MF_ShinyStar_Beam}.Result) {r=true}) {name=Line01_Mask, nodeColor=<rgba 1 0 0.0784 1>}
  n22 = Multiply(
   A=Multiply(A=NamedRerouteUsage(), B=ScalarParameter() {defaultValue=0.3, group=ShinyStar, parameterName=EmissiveIntensity})
   B=NamedRerouteUsage()
  )
  n23 = Multiply(
   A=Constant3Vector() {constant=<rgba 1 0.9702 0.7396 0>}
   B=ScalarParameter() {defaultValue=15, group=ShinyStar, parameterName=PanningStarEmissive}
  )
  n24 = NamedRerouteUsage()
  n25 = Add(A=Multiply(A=LinearInterpolate(A=@n22, B=@n23, Alpha=@n24), B=NamedRerouteUsage()), B=NamedRerouteUsage()) {desc=Center star emissive}
  NamedRerouteDeclaration(None=StaticSwitchParameter(True=MaterialFunctionCall!(In=@n25) {materialFunction=/Game/Weapons/WeaponSkins/MaterialFunction/BoostEmissivebyTOD}.Result, False=@n25) {group=Skin FX, parameterName=Use TOD Emissive Boost}) {name=Emissive_OUT, nodeColor=<rgba 1 0 0.8438 1>}
  n26 = ScalarParameter() {defaultValue=-2.5, group=ShinyStar, parameterName=BumpOffset_02}
  n27 = ScalarParameter() {defaultValue=0.05, group=ShinyStar, parameterName=ShinyStar_Beam_RotateSpeed_02}
  n28 = Multiply(A=@n16.Result) {constB=-1}
  n29 = NamedRerouteUsage()
  n30 = NamedRerouteUsage()
  n31 = NamedRerouteUsage()
  n32 = NamedRerouteUsage()
  NamedRerouteDeclaration(None=ComponentMask(None=MaterialFunctionCall(BumpOffsetMask=@n11, BumpOffset=@n26, RotateSpeed=@n27, UV_Distortion=@n28, UV_Offset=@n29, Time=@n30, Tex=@n19, DDX=@n31, DDY=@n32) {materialFunction=&MF_ShinyStar_Beam}.Result) {g=true}) {name=Line02_Mask, nodeColor=<rgba 0.0125 1 0 1>}
  n33 = ScalarParameter() {group=ShinyStar, parameterName=BumpOffset_03}
  n34 = ScalarParameter() {defaultValue=0.03, group=ShinyStar, parameterName=ShinyStar_Beam_RotateSpeed_03}
  n35 = Multiply(A=@n16.Result) {constB=0.5}
  n36 = NamedRerouteUsage()
  n37 = NamedRerouteUsage()
  n38 = NamedRerouteUsage()
  n39 = NamedRerouteUsage()
  NamedRerouteDeclaration(None=ComponentMask(None=MaterialFunctionCall(BumpOffsetMask=@n11, BumpOffset=@n33, RotateSpeed=@n34, UV_Distortion=@n35, UV_Offset=@n36, Time=@n37, Tex=@n19, DDX=@n38, DDY=@n39) {materialFunction=&MF_ShinyStar_Beam}.Result) {b=true}) {name=Line03_Mask, nodeColor=<rgba 0 0.0387 1 1>}
  n40 = ScalarParameter() {defaultValue=0.35, group=ShinyStar, parameterName=ShinyStarBackgroundDarknessBias}
  n41 = OneMinus(None=Saturate(None=@n3))
  NamedRerouteDeclaration(None=LinearInterpolate(B=Add(A=@n40, B=@n41), Alpha=NamedRerouteUsage()) {constA=1}) {name=Brightness, nodeColor=<rgba 1 0.9609 0 1>}
  NamedRerouteDeclaration(None=DDY(Value=@n9)) {name=_ddy, nodeColor=<rgba 1 0.8672 0 1>}
  n42 = NamedRerouteUsage()
  n43 = VectorParameter() {
   parameterName=StarPanningParams_01
   channelNames {a=TimeOffset, b=RotationSpeed, g=Direction_Y, r=Direction_X}
   defaultValue=<rgba 1 1 1.2 0>
  }
  n44 = ScalarParameter() {defaultValue=0.15, group=ShinyStar, parameterName=StarPanningSpeed}
  n45 = NamedRerouteUsage()
  n46 = MaterialFunctionCall(Float3=MaterialFunctionCall!(Reactivity=@n42, RotationSpeed=@n43.RotationSpeed, Direction=@n43.RotationSpeed, Speed=@n44, Offset=@n43.RotationSpeed, Tex=@n45) {materialFunction=&/MF_ShinyStar_StarEmitter}.Result) {materialFunction=&BreakOutFloat3Components}
  n47 = VectorParameter() {
   group=ShinyStar
   parameterName=StarPanningParams_02
   channelNames {a=TimeOffset, b=Speed_Rotation, g=Speed_Y, r=Speed_X}
   defaultValue=<rgba 1 -1 -1 0.2>
  }
  n48 = MaterialFunctionCall(Float3=MaterialFunctionCall(Reactivity=@n42, RotationSpeed=@n47.Speed_Rotation, Direction=@n47.Speed_Rotation, Speed=@n44, Offset=@n47.Speed_Rotation, Tex=@n45) {materialFunction=&MF_ShinyStar_StarEmitter}.Result) {materialFunction=&BreakOutFloat3Components}
  n49 = VectorParameter() {
   group=ShinyStar
   parameterName=StarPanningParams_03
   channelNames {a=TimeOffset, b=Speed_Rotation, g=Speed_Y, r=Speed_X}
   defaultValue=<rgba -0.3 -0.45 1 0.5>
  }
  n50 = MaterialFunctionCall(Reactivity=@n42, RotationSpeed=@n49.Speed_Rotation, Direction=@n49.Speed_Rotation, Speed=@n44, Offset=@n49.Speed_Rotation, Tex=@n45) {materialFunction=&MF_ShinyStar_StarEmitter}
  n51 = VectorParameter() {
   group=ShinyStar
   parameterName=StarPanningParams_04
   channelNames {a=TimeOffset, b=Speed_Rotation, g=Speed_Y, r=Speed_X}
   defaultValue=<rgba -1 0.1 -1 0.9>
  }
  n52 = Add(
   A=Add(
    A=Add(A=@n46.G, B=@n48.G)
    B=MaterialFunctionCall(Float3=@n50.Result) {materialFunction=&BreakOutFloat3Components}.G
   )
   B=MaterialFunctionCall(Float3=MaterialFunctionCall(Reactivity=@n42, RotationSpeed=@n51.Speed_Rotation, Direction=@n51.Speed_Rotation, Speed=@n44, Offset=@n51.Speed_Rotation, Tex=@n45) {materialFunction=&MF_ShinyStar_StarEmitter}.Result) {materialFunction=&BreakOutFloat3Components}.G
  )
  n53 = VectorParameter() {
   group=ShinyStar
   parameterName=StarPanningParams_05
   channelNames {a=TimeOffset, b=Speed_Rotation, g=Speed_Y, r=Speed_X}
   defaultValue=<rgba -0.5 1 -1 0.68>
  }
  n54 = MaterialFunctionCall(Float3=MaterialFunctionCall(Reactivity=@n42, RotationSpeed=@n53.Speed_Rotation, Direction=@n53.Speed_Rotation, Speed=@n44, Offset=@n53.Speed_Rotation, Tex=@n45) {materialFunction=&MF_ShinyStar_StarEmitter}.Result) {materialFunction=&BreakOutFloat3Components}
  NamedRerouteDeclaration(None=Multiply(A=Add(A=@n52, B=@n54.G), B=NamedRerouteUsage())) {name=PanningStarsMask, nodeColor=<rgba 0 1 0.5547 1>}
  n55 = SphereMask(
   A=TextureCoordinate()
   B=NamedRerouteUsage()
   Radius=ScalarParameter() {defaultValue=0.22, group=ShinyStar, parameterName=InnerCircleRadius}
   Hardness=ScalarParameter() {group=ShinyStar, parameterName=InnerCircleHardness}
  ) {attenuationRadius=0.2}
  NamedRerouteDeclaration(None=@n55) {name=BackgroundSoftGradient, nodeColor=<rgba 0.5703 0 1 1>}
  n56 = ScalarParameter() {defaultValue=0.6, group=ShinyStar, parameterName=ShinyStar_Beams_Speed}
  NamedRerouteDeclaration(None=Multiply(A=Time(), B=Multiply(A=@n56) {constB=-1})) {name=_beamTime, nodeColor=<rgba 0 1 0.6953 1>}
  n57 = VectorParameter() {group=ShinyStar, parameterName=ShinyStar_MainColor, defaultValue=<rgba 0.0445 0.0445 0.117 1>}
  NamedRerouteDeclaration(None=@n57.RGB) {name=MainColor, nodeColor=<rgba 1 0.8203 0 1>}
  n58 = NamedRerouteUsage()
  n59 = MaterialFunctionCall(Hue Shift Percentage=ScalarParameter() {group=ShinyStar, parameterName=OutterRing_Hueshift}, Texture=@n58) {materialFunction=&HueShift}
  n60 = ScalarParameter() {defaultValue=-1, group=ShinyStar, parameterName=OutterRing_Saturation}
  n61 = LinearInterpolate(
   A=@n58
   B=Multiply(
    A=Desaturation(None=@n59.Result, Fraction=@n60)
    B=ScalarParameter() {defaultValue=0.8, group=ShinyStar, parameterName=OutterRing_Brightness}
   )
   Alpha=NamedRerouteUsage()
  )
  n62 = VectorParameter() {group=ShinyStar, parameterName=ShinyStar_SecondaryColor, defaultValue=<rgba 1 0.0625 0.4397 1>}
  n63 = LinearInterpolate(
   A=LinearInterpolate(
    A=LinearInterpolate(A=@n57.RGB, B=@n62.RGB, Alpha=@n55)
    B=VectorParameter() {group=ShinyStar, parameterName=Line01_Color, defaultValue=<rgba 1 0.1683 0.5029 1>}.RGB
    Alpha=NamedRerouteUsage()
   )
   B=VectorParameter() {group=ShinyStar, parameterName=Line02_Color, defaultValue=<rgba 0.0091 0.776 0.3152 1>}.RGB
   Alpha=NamedRerouteUsage()
  )
  n64 = VectorParameter() {group=ShinyStar, parameterName=Line03_Color, defaultValue=<rgba 1 0.9473 0.3372 1>}
  n65 = NamedRerouteUsage()
  n66 = LinearInterpolate(
   A=Multiply(A=LinearInterpolate(A=@n63, B=@n64.RGB, Alpha=@n65), B=NamedRerouteUsage())
   B=VectorParameter() {group=ShinyStar, parameterName=PanningStarsColor, defaultValue=<rgba 0.8438 0.8438 0.8438 1>}.RGB
   Alpha=NamedRerouteUsage()
  )
  n67 = NamedRerouteUsage()
  NamedRerouteDeclaration(None=NamedRerouteDeclaration(None=LinearInterpolate(A=@n61, B=@n66, Alpha=@n67)) {name=FinalBaseColor, nodeColor=<rgba 0 1 0.0156 1>}) {name=BaseColor_OUT, nodeColor=<rgba 0.8281 0 1 1>}
  n68 = TextureSample(Tex=@n10) {
   samplerSource=SSM_Wrap_WorldGroupSettings
   samplerType=SAMPLERTYPE_Masks
   texture=&T_Backpack_ShinyStar_TA
  }
  NamedRerouteDeclaration(None=@n68.B) {name=CenterStar, nodeColor=<rgba 0.5703 0 1 1>}
  NamedRerouteDeclaration(None=@n1.Roughness) {name=Roughness_IN, nodeColor=<rgba 0.2656 0 1 1>}
  n69 = TextureCoordinate()
  n70 = NamedRerouteUsage()
  n71 = Add(A=Power(Base=SphereMask(A=@n69, B=@n70) {attenuationRadius=0.15, hardnessPercent=0}) {constExponent=3.5}, B=NamedRerouteUsage())
  n72 = VectorParameter() {parameterName=CenterReactivityColor, defaultValue=<rgba 1.3732 1.75 1.5073 1>}
  NamedRerouteDeclaration(None=LinearInterpolate(B=Multiply(A=@n71, B=@n72.RGB), Alpha=NamedRerouteUsage())) {name=CenterStarColor, nodeColor=<rgba 1 0.375 0 1>}
  n73 = NamedRerouteUsage()
  n74 = Multiply(A=@n73, B=ScalarParameter() {defaultValue=2.3, parameterName=ShinyStarRoughnessMulti})
  n75 = NamedRerouteUsage()
  NamedRerouteDeclaration(None=Saturate(None=LinearInterpolate(A=@n73, B=@n74, Alpha=@n75))) {name=Roughness_OUT, nodeColor=<rgba 1 0 0.4219 1>}
  n76 = ScalarParameter() {defaultValue=1, group=ShinyStar, parameterName=ReactivityToggle, sliderMax=1}
  NamedRerouteDeclaration(None=Multiply(A=ScalarParameter() {group=ShinyStar, parameterName=ReactivityBoost, sliderMax=2}, B=Clamp(None=@n76))) {name=FinalReactivity, nodeColor=<rgba 0 1 0.9297 1>}
  ScalarParameter() {group=ShinyStar, parameterName=ShinyStarBackgroundDarken}
  Subtract(A=Subtract(A=TextureCoordinate(), B=Constant() {r=0.5}), B=NamedRerouteUsage()) {desc=Offseted coordinates}
 ]}
