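# Master material for the ShinyStar pickaxe cosmetic, using Default Lit with material attributes and skeletal mesh support.
# The character PreFX function feeds static-switched Generic Gem, Thin Film and Metal LUT layers, driven by a SkinFX mask that defaults to the ShinyStar backpack's texture.
# The result passes through anisotropic shading and the mask-driven MF_ShinyStar effect, then character PostFX and cinematic PostFX functions.
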Material /BRCosmetics/Weapons/FORT_Melee/Pickaxe_ShinyStar/Materials/M_ShinyStar_MASTER
graph {
 properties {
  bUseMaterialAttributes=true
  bUsedWithClothing=true
  bUsedWithMeshDeformer=true
  bUsedWithSkeletalMesh=true
  pixelDepthOffsetMode=PDOM_Legacy
  usedShadingModels=Default Lit
 }
 nodes [
  n1 = MaterialFunctionCall() {materialFunction=&BaseFN_CharacterFunction_PreFX}
  n2 = TextureObjectParameter() {
   group=Skin FX
   parameterName=SkinFX_Mask
   samplerType=SAMPLERTYPE_Masks
   texture=/BRCosmetics/Accessories/FORT_Backpacks/Backpack_F_MED_ShinyStar/Textures/T_F_MED_ShinyStar_Backpack_FX
  }
  n3 = TextureSample(Tex=@n2)
  n4 = StaticSwitchParameter(
   True=MaterialFunctionCall!(
    MASK=StaticComponentMaskParameter(None=@n3.RGBA) {defaultR=true, group=Skin FX, parameterName=Generic Gem Effect Channel}
    Color Change Sequence=StaticComponentMaskParameter(None=@n3.RGBA) {defaultB=true, group=Skin FX, parameterName=Generic Gem Color Change Channel}
    Material=@n1.Result
   ) {materialFunction=/Game/Weapons/WeaponSkins/MaterialFunction/GenericGem}.Result
   False=@n1.Result
  ) {group=Skin FX, parameterName=Use Generic Gem}
  n5 = StaticSwitchParameter(
   True=MaterialFunctionCall!(Alpha=StaticComponentMaskParameter(None=@n3.RGBA) {defaultR=true, group=Skin FX, parameterName=ThinFilm_Channel}, Material Attributes=@n4) {materialFunction=/Game/Weapons/WeaponSkins/MaterialFunction/ThinFilmFunction}.Result
   False=@n4
  ) {group=Skin FX, parameterName=Use Thin Film}
  n6 = StaticSwitchParameter(
   True=MaterialFunctionCall(Mask=@n3.RGBA, Materials=@n5) {materialFunction=&MF_MetalsLUT}.Result
   False=@n5
  ) {group=Skin FX, parameterName=UseMetalLUT}
  n7 = MaterialFunctionCall(
   AO=@n1.AO
   Distance Blend=@n1.AO
   Fresnel=@n1.AO
   Gmap=@n1.AO
   Pre Skinned Local Position=@n1.AO
   Material=MaterialFunctionCall!(SkinFX=@n3.RGBA, In=MaterialFunctionCall!(In=@n6) {materialFunction=/Game/Weapons/WeaponSkins/MaterialFunction/MF_AnisotropicShading}.Result) {materialFunction=&/MF_ShinyStar}.Result
  ) {materialFunction=&BaseFN_CharacterFunction_PostFX}
  MaterialFunctionCall(In=@n7.Result) {materialFunction=&CINE_MF_CharPostFX}
 ]
}
